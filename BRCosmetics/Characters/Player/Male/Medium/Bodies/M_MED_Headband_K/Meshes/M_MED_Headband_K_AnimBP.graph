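# Post-pose Animation Blueprint for the M_MED_Headband_K body mesh, following the attached parent's pose and adding secondary motion to a ninja tool pouch.
# The copied parent pose feeds a constraint that places the pouch base by weighting pelvis, spine and thigh twist bones, mostly the pelvis.
# A heavily damped, angle-limited Anim Dynamics box then sways the pouch parent bone at partial alpha.
# A RigidBody node is left unconnected and the event graph has no logic.
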AnimBlueprint /BRCosmetics/Characters/Player/Male/Medium/Bodies/M_MED_Headband_K/Meshes/M_MED_Headband_K_AnimBP
target_skeleton=&/M_MED_Headband_K_Skeleton
graphs [
 AnimGraph {nodes [
   RigidBody() {
    bClampLinearTranslationLimitToRefPose=true
    bOverrideWorldGravity=true
    simulationSpace=BaseBoneSpace
    alphaBoolBlend {blendOption=Linear}
    baseBoneRef {boneName=spine_05}
    componentAppliedLinearAccClamp=<6000 6000 6000>
    componentLinearAccScale=<0.15 0.15 0.15>
    componentLinearVelScale=<0.15 0.15 0.15>
    overrideWorldGravity {z=-980}
   }
   n1 = "Constraint - dyn_ninja_tool_pouch_base (PPPPPPP)"!(
    ComponentPose=Local To Component(LocalPose=Copy Pose From Mesh() {bUseAttachedParent=true})
    Alpha=1
    ConstraintWeights_0=0.0336
    ConstraintWeights_1:float=0.1734
    ConstraintWeights_2:float=0.0168
    ConstraintWeights_3:float=0.0088
    ConstraintWeights_4:float=0.0005
    ConstraintWeights_5:float=0.7314
    ConstraintWeights_6:float=0.0356
   ) -> (Pose) {
    alphaBoolBlend {blendOption=Linear}
    boneToModify {boneName=dyn_ninja_tool_pouch_base}
    constraintSetup [
     {targetBone {boneName=thigh_twist_01_r}}
     {targetBone {boneName=spine_01}}
     {targetBone {boneName=spine_02}}
     {targetBone {boneName=spine_03}}
     {targetBone {boneName=spine_04}}
     {targetBone {boneName=pelvis}}
     {targetBone {boneName=thigh_twist_01_l}}
    ]
    constraintWeights [1,1,1,1,1,1,1]
   }
   n2 = Anim Dynamics - Bone: dyn_ninja_tool_pouch_parent(ComponentPose=@n1, Alpha=0.35, ExternalForce="0,0,0") {
    angularDampingOverride=0.9999
    bGravityOverrideInSimSpace=true
    bOverrideAngularDamping=true
    bOverrideLinearDamping=true
    linearDampingOverride=0.9999
    alphaBoolBlend {blendOption=Linear}
    boundBone {boneName=dyn_ninja_tool_pouch_parent}
    componentAppliedLinearAccClamp=<20000 20000 20000>
    componentLinearAccScale=<0.1 0.1 0.1>
    componentLinearVelScale=<0.1 0.1 0.1>
    physicsBodyDefinitions [{
      sphereCollisionRadius=0
      boundBone {boneName=dyn_ninja_tool_pouch_parent}
      boxExtents=<16 16 8>
      constraintSetup {angularLimitsMax=<10 0 30>, angularLimitsMin=<-10 0 -20>, angularTarget=<0 0 0>}
      localJointOffset=<5 0 -3>
     }]
    relativeSpaceBone {boneName=root}
   }
   Output Pose(Result=Component To Local(ComponentPose=@n2)) {layerGroup=None}
  ]}
 EventGraph {nodes [Event BlueprintUpdateAnimation()]}
]
variables []
