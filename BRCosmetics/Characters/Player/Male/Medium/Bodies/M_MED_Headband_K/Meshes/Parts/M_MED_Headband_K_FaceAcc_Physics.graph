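# Physics asset for the FaceAcc part of the male medium Headband K player cosmetic, covering its dynamic hair and headband bones.
# Thirteen hair pairs each constrain a capsule on the parent bone to a small sphere on the base bone.
# Two tie capsules share the knot sphere as their joint partner, while the hair offset sphere has no constraint.
# Tight swing and twist limits keep any motion subtle.
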PhysicsAsset /BRCosmetics/Characters/Player/Male/Medium/Bodies/M_MED_Headband_K/Meshes/Parts/M_MED_Headband_K_FaceAcc_Physics
bodies {
 dyn_hair_01_base [{shapeType=Sphere, radius=0.5, shapeName=dyn_hair_01_base_sphere}]
 dyn_hair_01_parent [{shapeType=Capsule, length=1.8022, radius=1.0363, shapeName=dyn_hair_01_parent_capsule, center=<1.087 0.0001 0>, rotation=<rot 90 0 0>}]
 dyn_hair_02_base [{shapeType=Sphere, radius=0.5, shapeName=dyn_hair_02_base_sphere}]
 dyn_hair_02_parent [{shapeType=Capsule, length=4.1237, radius=2.2855, shapeName=dyn_hair_02_parent_capsule, center=<2.3673 -0.0001 0>, rotation=<rot 90 0 0>}]
 dyn_hair_03_base [{shapeType=Sphere, radius=0.5, shapeName=dyn_hair_03_base_sphere}]
 dyn_hair_03_parent [{shapeType=Capsule, length=2.6633, radius=2.4587, shapeName=dyn_hair_03_parent_capsule, center=<1.7031 0 0>, rotation=<rot 90 0 0>}]
 dyn_hair_04_base [{shapeType=Sphere, radius=0.5, shapeName=dyn_hair_04_base_sphere}]
 dyn_hair_04_parent [{shapeType=Capsule, length=3.1043, radius=1.9322, shapeName=dyn_hair_04_parent_capsule, center=<1.838 0 0>, rotation=<rot 90 0 0>}]
 dyn_hair_05_base [{shapeType=Sphere, radius=0.5, shapeName=dyn_hair_05_base_sphere}]
 dyn_hair_05_parent [{shapeType=Capsule, length=2.1007, radius=1.4848, shapeName=dyn_hair_05_parent_capsule, center=<1.7502 0 -0.0001>, rotation=<rot 90 0 0>}]
 dyn_hair_06_base [{shapeType=Sphere, radius=0.5, shapeName=dyn_hair_06_base_sphere}]
 dyn_hair_06_parent [{shapeType=Capsule, length=4.0944, radius=2.3379, shapeName=dyn_hair_06_parent_capsule, center=<1.7193 0 0>, rotation=<rot 90 0 0>}]
 dyn_hair_07_base [{shapeType=Sphere, radius=0.5, shapeName=dyn_hair_07_base_sphere}]
 dyn_hair_07_parent [{shapeType=Capsule, length=2.2732, radius=1.832, shapeName=dyn_hair_07_parent_capsule, center=<1.815 0 0>, rotation=<rot 90 0 0>}]
 dyn_hair_08_base [{shapeType=Sphere, radius=0.5, shapeName=dyn_hair_08_base_sphere}]
 dyn_hair_08_parent [{shapeType=Capsule, length=2.5686, radius=1.4874, shapeName=dyn_hair_08_parent_capsule, center=<1.7484 0 0>, rotation=<rot 90 0 0>}]
 dyn_hair_09_base [{shapeType=Sphere, radius=0.5, shapeName=dyn_hair_09_base_sphere}]
 dyn_hair_09_parent [{shapeType=Capsule, length=3.4639, radius=3.1541, shapeName=dyn_hair_09_parent_capsule, center=<2.1979 0 0>, rotation=<rot 90 0 0>}]
 dyn_hair_10_base [{shapeType=Sphere, radius=0.5, shapeName=dyn_hair_10_base_sphere}]
 dyn_hair_10_parent [{shapeType=Capsule, length=1.5523, radius=1.452, shapeName=dyn_hair_10_parent_capsule, center=<0.8924 0 0>, rotation=<rot 90 0 0>}]
 dyn_hair_11_base [{shapeType=Sphere, radius=0.5, shapeName=dyn_hair_11_base_sphere}]
 dyn_hair_11_parent [{shapeType=Capsule, length=2.429, radius=1.457, shapeName=dyn_hair_11_parent_capsule, center=<1.2302 0 0>, rotation=<rot 90 0 0>}]
 dyn_hair_12_base [{shapeType=Sphere, radius=0.5, shapeName=dyn_hair_12_base_sphere}]
 dyn_hair_12_parent [{shapeType=Capsule, length=1.3947, radius=1.4452, shapeName=dyn_hair_12_parent_capsule, center=<1.3662 0 0>, rotation=<rot 90 0 0>}]
 dyn_hair_13_base [{shapeType=Sphere, radius=0.5, shapeName=dyn_hair_13_base_sphere}]
 dyn_hair_13_parent [{shapeType=Capsule, length=1.7987, radius=1.7178, shapeName=dyn_hair_13_parent_capsule, center=<0.9389 0 0>, rotation=<rot 90 0 0>}]
 dyn_hair_offset [{shapeType=Sphere, radius=0.5, shapeName=dyn_hair_offset_sphere}]
 dyn_headband_knot [{shapeType=Sphere, radius=0.5, shapeName=dyn_headband_knot_sphere}]
 dyn_headband_tie_01 [{shapeType=Capsule, length=3.6191, radius=1.7867, shapeName=dyn_headband_tie_01_capsule, center=<2.5828 0 0>, rotation=<rot 90 0 0>}]
 dyn_headband_tie_02 [{shapeType=Capsule, length=3.7261, radius=1.5754, shapeName=dyn_headband_tie_02_capsule, center=<2.9845 0 0.0002>, rotation=<rot 90 0 0>}]
}
constraints [
 {bone1Name=dyn_hair_13_parent, bone2Name=dyn_hair_13_base, swing1LimitDegrees=10, swing2LimitDegrees=10, twistLimitDegrees=3}
 {bone1Name=dyn_hair_12_parent, bone2Name=dyn_hair_12_base, swing1LimitDegrees=6, swing2LimitDegrees=6, twistLimitDegrees=3}
 {bone1Name=dyn_hair_11_parent, bone2Name=dyn_hair_11_base, swing1LimitDegrees=10, swing2LimitDegrees=10, twistLimitDegrees=3}
 {bone1Name=dyn_hair_10_parent, bone2Name=dyn_hair_10_base, swing1LimitDegrees=10, swing2LimitDegrees=10, twistLimitDegrees=3}
 {bone1Name=dyn_hair_09_parent, bone2Name=dyn_hair_09_base, swing1LimitDegrees=10, swing2LimitDegrees=10, twistLimitDegrees=3}
 {bone1Name=dyn_hair_08_parent, bone2Name=dyn_hair_08_base, swing1LimitDegrees=10, swing2LimitDegrees=10, twistLimitDegrees=3}
 {bone1Name=dyn_hair_07_parent, bone2Name=dyn_hair_07_base, swing1LimitDegrees=6, swing2LimitDegrees=6, twistLimitDegrees=3}
 {bone1Name=dyn_hair_06_parent, bone2Name=dyn_hair_06_base, swing1LimitDegrees=6, swing2LimitDegrees=6, twistLimitDegrees=3}
 {bone1Name=dyn_hair_05_parent, bone2Name=dyn_hair_05_base, swing1LimitDegrees=5, swing2LimitDegrees=5, twistLimitDegrees=3}
 {bone1Name=dyn_hair_04_parent, bone2Name=dyn_hair_04_base, swing1LimitDegrees=6, swing2LimitDegrees=6, twistLimitDegrees=3}
 {bone1Name=dyn_hair_03_parent, bone2Name=dyn_hair_03_base, swing1LimitDegrees=10, swing2LimitDegrees=10, twistLimitDegrees=3}
 {bone1Name=dyn_hair_02_parent, bone2Name=dyn_hair_02_base, swing1LimitDegrees=10, swing2LimitDegrees=10, twistLimitDegrees=3}
 {bone1Name=dyn_hair_01_parent, bone2Name=dyn_hair_01_base, swing1LimitDegrees=10, swing2LimitDegrees=10, twistLimitDegrees=3}
 {bone1Name=dyn_headband_tie_02, bone2Name=dyn_headband_knot, swing1LimitDegrees=10, swing2LimitDegrees=10, twistLimitDegrees=3}
 {bone1Name=dyn_headband_tie_01, bone2Name=dyn_headband_knot, swing1LimitDegrees=6, swing2LimitDegrees=6, twistLimitDegrees=3}
]
